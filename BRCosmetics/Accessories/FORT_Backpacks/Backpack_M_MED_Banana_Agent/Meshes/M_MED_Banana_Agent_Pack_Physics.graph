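# Physics asset for the Banana Agent backpack cosmetic, defining simulation bodies for its banana, bag-opening and ammo bones.
# Each bone gets a single box body.
# Five banana pieces and two bag-opening pieces constrain directly to dyn_parent in a star layout, while dyn_ammo_parent hangs separately from dyn_ammo_base.
# Tight swing and twist limits keep movement to a small wobble, with the ammo piece allowed the widest swing.
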PhysicsAsset /BRCosmetics/Accessories/FORT_Backpacks/Backpack_M_MED_Banana_Agent/Meshes/M_MED_Banana_Agent_Pack_Physics
bodies {
 dyn_ammo_base [{shapeType=Box, extentX=2.9375, extentY=2.9375, extentZ=2.9375}]
 dyn_ammo_parent [{shapeType=Box, extentX=4.3816, extentY=8.6322, extentZ=2.5375, center=<1.9025 0 0>}]
 dyn_bag_open_bottom [{shapeType=Box, extentX=4, extentY=7.8223, extentZ=3.0772, center=<0.8574 0 0.0001>}]
 dyn_bag_open_top [{shapeType=Box, extentX=4, extentY=7.8223, extentZ=2.1397, center=<0.5316 0 0>}]
 dyn_banana_back_bottom [{shapeType=Box, extentX=7.506, extentY=4.0257, extentZ=5.6911, center=<1.9566 -0.0001 0>}]
 dyn_banana_back_top [{shapeType=Box, extentX=12.2922, extentY=6.7124, extentZ=6.3754, center=<6.7145 0 0>}]
 dyn_banana_middle_top [{shapeType=Box, extentX=12.4601, extentY=6.9426, extentZ=7.6021, center=<4.2552 0 0>}]
 dyn_main_banana_bottom [{shapeType=Box, extentX=15.6524, extentY=11.6301, extentZ=8.7293, center=<11.1802 0 -4.7919>, rotation=<rot -39.3754 0 0>}]
 dyn_main_banana_top [{shapeType=Box, extentX=16.8234, extentY=8.6027, extentZ=8.6641, center=<10.1524 0 0>}]
 dyn_parent [{shapeType=Box, extentX=7.5, extentY=7.5, extentZ=7.5}]
}
constraints [
 {bone1Name=dyn_main_banana_bottom, bone2Name=dyn_parent, swing1LimitDegrees=5, swing2LimitDegrees=10, twistLimitDegrees=3}
 {bone1Name=dyn_main_banana_top, bone2Name=dyn_parent, swing1LimitDegrees=10, swing2LimitDegrees=10, twistLimitDegrees=3}
 {bone1Name=dyn_banana_middle_top, bone2Name=dyn_parent, swing1LimitDegrees=10, swing2LimitDegrees=10, twistLimitDegrees=3}
 {bone1Name=dyn_banana_back_top, bone2Name=dyn_parent, swing1LimitDegrees=10, swing2LimitDegrees=10, twistLimitDegrees=3}
 {bone1Name=dyn_banana_back_bottom, bone2Name=dyn_parent, swing1LimitDegrees=10, swing2LimitDegrees=10, twistLimitDegrees=3}
 {bone1Name=dyn_bag_open_top, bone2Name=dyn_parent, swing1LimitDegrees=5, swing2LimitDegrees=5, twistLimitDegrees=3}
 {bone1Name=dyn_bag_open_bottom, bone2Name=dyn_parent, swing1LimitDegrees=5, swing2LimitDegrees=5, twistLimitDegrees=3}
 {bone1Name=dyn_ammo_parent, bone2Name=dyn_ammo_base, swing1LimitDegrees=15, swing2LimitDegrees=20, twistLimitDegrees=2}
]
