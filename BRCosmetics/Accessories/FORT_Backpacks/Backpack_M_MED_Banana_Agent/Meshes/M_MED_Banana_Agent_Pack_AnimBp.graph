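# Animation Blueprint for the Banana Agent backpack cosmetic that adds secondary physics motion to the bag and its zippers.
# The input pose runs through a RigidBody simulation in root-bone space, using the pack's physics asset with overridden gravity.
# Constraint nodes then position the zipper bases, blending the front one between the bag's open top and pack and pinning the side one to pack.
# Each constraint feeds an angle-limited, heavily damped AnimDynamics node for that zipper; the event graph does nothing.
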AnimBlueprint /BRCosmetics/Accessories/FORT_Backpacks/Backpack_M_MED_Banana_Agent/Meshes/M_MED_Banana_Agent_Pack_AnimBp
target_skeleton=&/M_MED_Banana_Agent_Pack_Skeleton
graphs [
 AnimGraph {nodes [
   n1 = RigidBody(ComponentPose=Local To Component(LocalPose=Input Pose())) {
    bClampLinearTranslationLimitToRefPose=true
    bOverrideWorldGravity=true
    overridePhysicsAsset=&/M_MED_Banana_Agent_Pack_Physics
    simulationSpace=BaseBoneSpace
    alphaBoolBlend {blendOption=Linear}
    baseBoneRef {boneName=root}
    componentAppliedLinearAccClamp=<3000 3000 3000>
    componentLinearAccScale=<0.25 0.25 0.25>
    componentLinearVelScale=<0.25 0.25 0.25>
    overrideWorldGravity {z=-980}
   }
   n2 = "Constraint - dyn_zipper_front_base (PP)"!(ComponentPose=@n1, Alpha=1, ConstraintWeights_0=0.4, ConstraintWeights_1:float=0.6) -> (Pose) {
    alphaBoolBlend {blendOption=Linear}
    boneToModify {boneName=dyn_zipper_front_base}
    constraintSetup [{targetBone {boneName=dyn_bag_open_top}}, {targetBone {boneName=pack}}]
    constraintWeights [1,1]
   }
   n3 = Anim Dynamics - Bone: dyn_zipper_front_parent(ComponentPose=@n2, Alpha=1, ExternalForce="0,0,0") {
    angularDampingOverride=0.999
    bGravityOverrideInSimSpace=true
    bOverrideAngularDamping=true
    bUsePlanarLimit=false
    gravityScale=0.75
    alphaBoolBlend {blendOption=Linear}
    boundBone {boneName=dyn_zipper_front_parent}
    componentAppliedLinearAccClamp=<6000 6000 6000>
    componentLinearAccScale=<0.15 0.15 0.15>
    componentLinearVelScale=<0.15 0.15 0.15>
    physicsBodyDefinitions [{
      sphereCollisionRadius=0
      boundBone {boneName=dyn_zipper_front_parent}
      boxExtents=<5 4 2>
      constraintSetup {angularLimitsMax=<0 15 80>, angularLimitsMin=<0 -15 -15>, angularTarget=<0 0 0>}
      localJointOffset=<2 0 0>
     }]
    relativeSpaceBone {boneName=root}
   }
   n4 = "Constraint - dyn_zipper_side_base (P)"!(ComponentPose=@n3, Alpha=1, ConstraintWeights_0=1, ConstraintWeights_1:float=0) -> (Pose) {
    alphaBoolBlend {blendOption=Linear}
    boneToModify {boneName=dyn_zipper_side_base}
    constraintSetup [{targetBone {boneName=pack}}]
    constraintWeights [1,0]
   }
   n5 = Anim Dynamics - Bone: dyn_zipper_side_parent(ComponentPose=@n4, Alpha=1, ExternalForce="0,0,0") {
    angularDampingOverride=0.999
    bGravityOverrideInSimSpace=true
    bOverrideAngularDamping=true
    bUsePlanarLimit=false
    gravityScale=0.75
    alphaBoolBlend {blendOption=Linear}
    boundBone {boneName=dyn_zipper_side_parent}
    componentAppliedLinearAccClamp=<6000 6000 6000>
    componentLinearAccScale=<0.15 0.15 0.15>
    componentLinearVelScale=<0.15 0.15 0.15>
    physicsBodyDefinitions [{
      sphereCollisionRadius=0
      boundBone {boneName=dyn_zipper_side_parent}
      boxExtents=<5 4 2>
      constraintSetup {angularLimitsMax=<0 50 60>, angularLimitsMin=<0 -50 -60>, angularTarget=<0 0 0>}
      localJointOffset=<2 0 0>
     }]
    relativeSpaceBone {boneName=root}
   }
   Output Pose(Result=Component To Local(ComponentPose=@n5)) {layerGroup=None}
  ]}
 EventGraph {nodes [
   TryGetPawnOwner()
   Event BlueprintUpdateAnimation()
  ]}
]
variables []
